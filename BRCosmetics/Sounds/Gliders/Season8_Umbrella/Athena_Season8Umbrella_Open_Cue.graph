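# Sound cue for the opening sound of the Season 8 Umbrella glider cosmetic in Athena.
# A Random node picks between two equally weighted wave variations.
# Shared glider open/close attenuation and concurrency assets, plus a cosmetic glider-open property sheet, govern playback.
# The cue is routed to the Fort_Gliders_Land sound class, despite being an open sound.
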SoundCue /BRCosmetics/Sounds/Gliders/Season8_Umbrella/Athena_Season8Umbrella_Open_Cue
root=Random_0
nodes [
 {id=Random_0, children [WavePlayer_0,WavePlayer_3], properties {weights [1,1]}}
 {id=WavePlayer_0, properties {soundWaveAssetPtr=&/SFX_Season8Umbrella_Open_01}}
 {id=WavePlayer_3, properties {soundWaveAssetPtr=&/SFX_Season8Umbrella_Open_02}}
]
properties {
 attenuationSettings=/Game/Sounds/Attenuation/Fort_Player_Glider_OpenClose_Athena_Att
 audioPropertiesSheet=/BRCosmetics/Sounds/Settings/PropertySheets/Glider/APS_Cosmetic_Glider_Open
 maxDistance=10000
 soundClassObject=&Fort_Gliders_Land
 volumeMultiplier=1
 concurrencySet [&Athena_GliderOpenClose_Conc]
}
